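# Physics asset for the male Nightmare pickaxe skeletal mesh, a two-handed melee weapon cosmetic.
# It holds a single capsule body on the Melee_TwoHanded_Handle bone, offset and rotated from the bone origin.
# It defines no constraints, so the weapon has one simple physics shape with no jointed parts.
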PhysicsAsset /BRCosmetics/Weapons/FORT_Melee/Pickaxe_Nightmare/Meshes/Nightmare_Male_Axe_PhysicsAsset
bodies {Melee_TwoHanded_Handle [{shapeType=Capsule, length=13.7305, radius=4.2713, center=<1.1688 -13.3725 -0.1631>, rotation=<rot 0 -179.6383 -89.4369>}]}
constraints []
